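# Translucent particle-sprite material for the NightNight final impact shockwave, tinted and faded by particle color and alpha.
# Two panning samples of an inky smoke texture use radial UVs, offset by dynamic-parameter pan and random values, and are colored by two chroma tints.
# Their product, shaped by sphere masks that hollow the center, drives opacity with optional depth fade.
# An optional world-position offset shrinks particles toward their centers near the camera.
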Material /Game/Athena/Prototype/Blueprints/NightNight/FX/FinalImpact/M_NightNight_ShockWave
graph {
 properties {bUsedWithParticleSprites=true, blendMode=BLEND_Translucent, pixelDepthOffsetMode=PDOM_Legacy, usedShadingModels=Default Lit}
 nodes [
  n1 = ParticleColor()
  n2 = TextureCoordinate()
  n3 = Constant() {r=0.5}
  n4 = AppendVector(
   A=MaterialFunctionCall() {materialFunction=&VectorToRadialValue}.Vector Converted to Angle
   B=Power(Base=SphereMask(A=@n2, B=@n3) {attenuationRadius=0.707, hardnessPercent=0}) {constExponent=2.5}
  )
  n5 = VectorParameter() {parameterName=Texture1_ctrl, defaultValue=<rgba 3 0.4 0 0.1>}
  n6 = AppendVector(A=@n5.B, B=@n5.B)
  n7 = Add(A=Multiply(A=Panner(Coordinate=@n4, Speed=@n6), B=AppendVector(A=@n5.R, B=@n5.R))) {constB=0.5}
  n8 = DynamicParameter() {paramNames [V_pan,UV_random,Param3,Param4]}
  n9 = AppendVector(A=Constant(), B=@n8.V_pan)
  n10 = TextureSample(UVs=Add(A=Add(A=@n7, B=@n9), B=@n8.UV_random)) {texture=&T_Inky_Smoke_Tile}
  n11 = VectorParameter() {parameterName=Texture2_ctrl, defaultValue=<rgba 2 0.5 0.02 0.1>}
  n12 = Multiply(A=Panner(Coordinate=@n4, Speed=AppendVector(A=@n11.B, B=@n11.B)), B=AppendVector(A=@n11.R, B=@n11.R))
  n13 = TextureSample(UVs=Add(A=@n8.UV_random, B=Add(A=@n9, B=@n12))) {texture=&T_Inky_Smoke_Tile}
  n14 = Add(
   A=Multiply(A=@n10.B, B=VectorParameter() {parameterName=Chroma1, defaultValue=<rgba 0.0659 0.8903 1 1>}.RGB)
   B=Multiply(A=@n13.B, B=VectorParameter() {parameterName=Chroma2, defaultValue=<rgba 1 0.6146 0.7047 1>}.RGB)
  )
  MaterialFunctionCall(Emissive=StaticSwitchParameter(True=@n1.RGB, False=Multiply(A=@n14, B=@n1.RGB)) {parameterName=UseParticleColor}) {materialFunction=&MF_FX_EmissiveAdjustments}
  n15 = Multiply(A=SphereMask(A=TextureCoordinate(), B=Constant() {r=0.5}) {attenuationRadius=0.5, hardnessPercent=1}) {constB=0.25}
  n16 = Multiply(A=Multiply(A=@n10.R, B=@n13.R), B=Power(Base=Sine(None=@n15)) {constExponent=3})
  n17 = SphereMask(A=TextureCoordinate(), B=Constant() {r=0.5}) {attenuationRadius=0.4, hardnessPercent=1}
  n18 = LinearInterpolate(Alpha=Power(Base=OneMinus(None=@n17))) {constA=-0.05}
  n19 = Multiply(A=@n1.A, B=SphereMask(A=Multiply(A=@n16, B=@n18), B=Constant() {r=0.5}) {attenuationRadius=0.5, hardnessPercent=80})
  n20 = DepthFade(Opacity=@n19, FadeDistance=ScalarParameter() {parameterName=DepthFade})
  Reroute(None=StaticSwitchParameter(True=ShadingPathSwitch(Default=@n20, Deferred=@n20, Mobile=@n19) {desc=Migrated from FeatureLevelSwitch}, False=@n19) {parameterName=Use_DepthFade})
  n21 = ParticlePositionWS()
  n22 = WorldPosition()
  n23 = Subtract(A=CameraPositionWS(), B=@n21)
  n24 = Clamp(None=Divide(
    A=MaterialFunctionCall(Vector 3=@n23) {materialFunction=/Engine/Functions/Engine_MaterialFunctions02/Utility/VectorLength}.V3 Length
    B=Constant() {r=120}
   ))
  StaticSwitchParameter(True=Multiply(A=Subtract(A=@n21, B=@n22.XYZ), B=LinearInterpolate(Alpha=@n24) {constA=1.1, constB=0}), False=Constant()) {parameterName=Use_Particle_Shrink}
 ]
}
