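# Additive unlit particle material for NightNight's FinalImpact FX, drawing flickering flare sprites.
# World position offset billboards sprites, sizing them by particle dynamic parameters, layered-sine flicker, near-camera fade and depth-based minimum screen width, plus camera offset.
# Vertex color red offsets flicker timing and varies size per sprite.
# Emissive blends two flare textures with color tints and HDR remapping, using simpler mobile paths; opacity has depth fade and optional random flicker.
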Material /Game/Athena/Prototype/Blueprints/NightNight/FX/FinalImpact/M_Sprites
graph {
 properties {
  bEnableResponsiveAA=true
  bUsedWithMeshParticles=true
  bUsedWithNiagaraMeshParticles=true
  bUsedWithNiagaraSprites=true
  bUsedWithParticleSprites=true
  blendMode=BLEND_Additive
  pixelDepthOffsetMode=PDOM_Legacy
  shadingModel=MSM_Unlit
  usedShadingModels=Unlit
 }
 nodes [
  n1 = DynamicParameter() {paramNames [BlendFlareTextures, StarFlickIntense, StarFlickSpeed, Size]}
  n2 = ScalarParameter() {defaultValue=0.8, parameterName=SpriteSquareSize}
  n3 = VertexColor()
  n4 = Multiply(
   A=Add(A=Time(), B=Multiply(A=@n3.R) {constB=5})
   B=LinearInterpolate(B=ScalarParameter() {defaultValue=8, parameterName=StarFlickerInitialSpeed}, Alpha=@n1.StarFlickSpeed)
  ) {constB=80}
  n5 = Sine(None=@n4) {period=0.155}
  n6 = Sine(None=@n4) {period=0.2}
  n7 = Multiply(A=Sine(None=@n4) {period=0.075}, B=Multiply(A=Multiply(A=@n5, B=@n6), B=Sine(None=@n4) {period=0.43}))
  n8 = Multiply(
   A=LinearInterpolate(A=ScalarParameter() {parameterName=MinSizeMult}, Alpha=VertexColor().R)
   B=LinearInterpolate(A=Multiply(A=@n2) {constB=0.5}, B=@n2, Alpha=LinearInterpolate(B=@n7, Alpha=@n1.StarFlickIntense) {constA=1})
  )
  n9 = MaterialFunctionCall(Distance=ScalarParameter() {defaultValue=40000, parameterName=WPO_NearCameraFade}) {materialFunction=&Near_Camera_Fade}
  n10 = ViewSize()
  n11 = Divide(A=PixelDepth().R, B=Multiply(A=ComponentMask(None=@n10) {g=true}, B=ScalarParameter() {defaultValue=0.08, parameterName=ScreenWidth}))
  n12 = Clamp(None=@n11, Max=ScalarParameter() {defaultValue=25, parameterName=MaxScreenWidthAdd})
  n13 = ScalarParameter() {parameterName=ClampMinScreenWidthAdd}
  n14 = Multiply(
   A=LinearInterpolate(
    A=@n1.Size
    B=ScalarParameter() {defaultValue=1, parameterName="SpritSize(OnlyWhenUseRandomOpacityEnabled)"}
    Alpha=ScalarParameter() {parameterName=UseRandomOpacity}
   )
   B=Add(A=LinearInterpolate(B=@n8, Alpha=@n9.Result), B=LinearInterpolate(A=@n11, B=@n12, Alpha=@n13))
  )
  ComponentMask(None=@n14) {r=true}
  n15 = ParticleColor()
  n16 = ShadingPathSwitch(Default=DepthFade(Opacity=@n15.A, FadeDistance=ScalarParameter() {parameterName=DepthFade}), Mobile=@n15.A)
  n17 = Time()
  n18 = DynamicParameter() {paramNames [BlendFlareTextures, StarFlickIntense, StarFlickSpeed, Size]}
  n19 = ScalarParameter() {desc=Random Opacity FLicker Offset to use vertex color, parameterName=Enable RandomOpacityFlicker Offset}
  n20 = Constant()
  n21 = Multiply(A=VertexColor().R) {constB=5}
  n22 = Multiply(
   A=Add(A=Add(A=@n17, B=@n18.Size), B=If(A=@n19, "A > B"=@n21, "A == B"=@n21, "A < B"=@n20) {constB=0.99})
   B=ScalarParameter() {defaultValue=0.005, parameterName=RandomOpacityFlickerSpeed}
  ) {constB=80}
  n23 = Sine(None=@n22) {period=0.075}
  n24 = Multiply(A=Multiply(A=Sine(None=@n22) {period=0.155}, B=Sine(None=@n22) {period=0.2}), B=Sine(None=@n22) {period=0.43})
  LinearInterpolate(A=@n16, B=Multiply(A=@n16, B=Multiply(A=@n23, B=@n24)), Alpha=ScalarParameter() {parameterName=UseRandomOpacity}) {constA=1}
  n25 = Constant() {r=1}
  n26 = LinearInterpolate(
   A=@n25
   B=Desaturation(None=VertexColor(), Fraction=ScalarParameter() {defaultValue=-0.5, parameterName=VertColorDesaturation})
   Alpha=ScalarParameter() {defaultValue=1, parameterName=UseVertextColorForRandomColor}
  )
  n27 = ShadingPathSwitch(Default=@n26, Deferred=@n26, Forward=@n25, Mobile=@n25)
  n28 = VectorParameter() {parameterName=Color, defaultValue=<rgba 0.7035 0.9089 1 1>}
  n29 = TextureSampleParameter2D() {parameterName=FlareA, texture=/Game/Effects/Fort_Effects/Textures/Flares/T_Flare_Spikeball}
  n30 = TextureSampleParameter2D() {parameterName=FlareB, texture=/Game/Effects/Fort_Effects/Textures/Flares/T_Flare_Soft}
  n31 = LinearInterpolate(A=@n29.RGB, B=@n30.RGB, Alpha=@n1.BlendFlareTextures)
  n32 = LinearInterpolate(A=@n29.R, B=@n30.R, Alpha=@n1.BlendFlareTextures)
  n33 = Multiply(A=Multiply(A=Multiply(A=@n27, B=@n28.RGB), B=@n15.RGB), B=ShadingPathSwitch(Default=@n31, Deferred=@n31, Forward=@n32, Mobile=@n32))
  n34 = MaterialFunctionCall(LDR Input=Clamp(None=@n33), HDR Tint=Multiply(A=@n33, B=ScalarParameter() {defaultValue=100, parameterName=Brightness}) {constB=100}) {materialFunction=&DeriveHDRfromLDR}
  MaterialFunctionCall(Emissive=ShadingPathSwitch(Default=@n34.RemappedRange, Deferred=@n34.RemappedRange, Forward=@n33, Mobile=@n33)) {materialFunction=&MF_FX_EmissiveAdjustments}
  n35 = WorldPosition()
  n36 = Multiply(A=MaterialFunctionCall() {materialFunction=&CameraDirectionVector}.Result) {constB=-1}
  n37 = ScalarParameter() {group=WPO, parameterName=Camera Offset}
  StaticSwitchParameter(
   True=Constant()
   False=Add(
    A=MaterialFunctionCall!(Center Location=@n35.XYZ, 0-1 UVs, XY scale=@n14, OverridePivot, "Normals (Optional)") {materialFunction=&Sprite}.World position offset
    B=Multiply(A=@n36, B=@n37)
   )
  ) {parameterName=TurnOff_WPO}
 ]
}
